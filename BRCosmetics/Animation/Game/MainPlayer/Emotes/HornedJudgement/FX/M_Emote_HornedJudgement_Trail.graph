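# Translucent, two-sided unlit Niagara material for the trail effect in the Horned Judgement emote.
# A tapered UV mask with distorted, panning wispy-flame noise drives opacity and a curve-atlas lookup that blends three tints of particle color.
# Dynamic parameters control UV offset, swipe, side fade and axis-driven bend, added to noise-based vertex displacement.
# Opacity adds view-angle fade, optional local-axis fade and edge hardening, with depth fade skipped on forward and mobile paths.
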Material /BRCosmetics/Animation/Game/MainPlayer/Emotes/HornedJudgement/FX/M_Emote_HornedJudgement_Trail
graph {
 properties {
  bUsedWithNiagaraMeshParticles=true
  bUsedWithNiagaraRibbons=true
  bUsedWithNiagaraSprites=true
  blendMode=BLEND_Translucent
  pixelDepthOffsetMode=PDOM_Legacy
  shadingModel=MSM_Unlit
  twoSided=true
  usedShadingModels=Unlit
 }
 nodes [
  n1 = TextureCoordinate()
  n2 = ComponentMask(None=VectorParameter() {group=WPO, parameterName=WPO_UVScale, defaultValue=<rgba 0 1 0 0>}.RGB) {g=true, r=true}
  n3 = DynamicParameter() {defaultValue=<rgba 0 0.75 0.8 1>, paramNames [UV Random Offset, Local Axis Fade Min, Side Fade, Swipe]}
  n4 = Constant()
  n5 = Panner(
   Coordinate=Add(A=Multiply(A=@n1, B=@n2), B=AppendVector(A=@n3.UV Random Offset, B=@n4)) {constB=0}
   Speed=ComponentMask(None=VectorParameter() {group=WPO, parameterName=WPO_Speed, defaultValue=<rgba 0 0.5 0 0>}.RGB) {g=true, r=true}
  ) {speedY=0.5}
  n6 = Power(Base=ComponentMask(None=TextureCoordinate()) {g=true}) {constExponent=1}
  n7 = Multiply(
   A=LinearInterpolate(Alpha=TextureSample(UVs=@n5) {samplerType=SAMPLERTYPE_Normal, texture=&T_Weap_Distort_N_02}.R) {constB=1.5}
   B=Saturate(None=Add(A=@n6) {constB=0.1})
  )
  n8 = VectorParameter() {group=WPO, parameterName=WPO_direction, defaultValue=<rgba 1 0 0 1>}
  n9 = DynamicParameter() {parameterIndex=2, paramNames [X,Y,Z,Param4]}
  n10 = ScalarParameter() {defaultValue=20, parameterName=WPO Max Offset, sliderMax=50}
  n11 = Multiply(
   A=VectorParameter() {parameterName=WPO Axis Mask, defaultValue=<rgba 1 1 1 1>}.RGB
   B=Clamp(
    None=MaterialFunctionCall(X=@n9.X, Y=@n9.X, Z=@n9.X) {materialFunction=&MakeFloat3}.Result
    Min=Multiply(A=@n10) {constB=-1}
    Max=@n10
   )
  )
  n12 = Power(Base=ComponentMask(None=TextureCoordinate()) {g=true}, Exp=ScalarParameter() {defaultValue=1, parameterName=WPO Bend, sliderMax=10})
  Add(
   A=Transform(None=Multiply(A=@n7, B=@n8.RGB))
   B=Multiply(A=ScalarParameter() {parameterName=WPO Intensity, sliderMax=100}, B=LinearInterpolate(B=@n11, Alpha=@n12))
  )
  n13 = ParticleColor()
  n14 = ScalarParameter() {defaultValue=2, group=Color, parameterName=Light Multiplier, sliderMax=5}
  n15 = ScalarParameter() {defaultValue=1, group=Color, parameterName=Middle Multiplier, sliderMax=5}
  n16 = Desaturation(
   None=MaterialFunctionCall(
    Hue Shift Percentage=ScalarParameter() {defaultValue=0.02, group=Color, parameterName=Dark Hue Shift, sliderMax=1, sliderMin=-1}
    Texture=@n13.RGB
   ) {materialFunction=&HueShift}.Result
   Fraction=ScalarParameter() {group=Color, parameterName=Desaturation, sliderMax=2, sliderMin=-2}
  )
  n17 = ScalarParameter() {defaultValue=0.5, group=Color, parameterName=Dark Multiplier, sliderMax=5}
  n18 = TextureCoordinate()
  n19 = StaticSwitchParameter(True=OneMinus(None=ComponentMask(None=@n18) {g=true}), False=OneMinus(None=ComponentMask(None=@n18) {r=true})) {parameterName=Swap UV}
  n20 = TextureCoordinate()
  n21 = ComponentMask(None=@n20) {r=true}
  n22 = ComponentMask(None=@n20) {g=true}
  n23 = DynamicParameter() {defaultValue=<rgba 0 0.75 0.8 1>, paramNames [UV Random Offset, Local Axis Fade Min, Side Fade, Swipe]}
  n24 = Power(
   Base=SphereMask(
    A=StaticSwitchParameter(True=@n21, False=@n22) {parameterName=Swap UV}
    B=Constant() {r=0.5}
    Radius=ScalarParameter() {defaultValue=0.5, parameterName=Width, sliderMax=1}
   ) {attenuationRadius=0.5, hardnessPercent=0}
   Exp=@n23.Side Fade
  )
  n25 = Power(Base=@n19, Exp=ScalarParameter() {parameterName=Start Fade Intensity, sliderMax=5}) {constExponent=30}
  n26 = Constant()
  n27 = SmoothStep(
   Min=ScalarParameter() {parameterName=Pointiness, sliderMax=1}
   Value=Subtract(A=Multiply(A=@n19, B=@n24), B=StaticSwitchParameter(True=@n25, False=@n26) {parameterName=Start Fade?})
  )
  n28 = Multiply(A=@n27, B=ScalarParameter() {defaultValue=1, parameterName=B, sliderMax=5}) {constB=3}
  n29 = TextureCoordinate()
  n30 = DynamicParameter() {defaultValue=<rgba 0 0.75 0.8 1>, paramNames [UV Random Offset, Local Axis Fade Min, Side Fade, Swipe]}
  n31 = Constant()
  n32 = Panner(
   Coordinate=Add(A=Add(A=@n29, B=@n30.UV Random Offset), B=AppendVector(A=@n30.Swipe, B=@n31))
   Speed=AppendVector(
    A=ScalarParameter() {defaultValue=-1, group=Wispy Flame, parameterName=Speed X, sliderMax=5, sliderMin=-5}
    B=ScalarParameter() {group=Wispy Flame, parameterName=Speed Y, sliderMax=5, sliderMin=-5}
   )
  ) {speedX=1}
  n33 = AppendVector(
   A=ScalarParameter() {defaultValue=1, group=Wispy Flame, parameterName=Scale X, sliderMax=5}
   B=ScalarParameter() {defaultValue=1, group=Wispy Flame, parameterName=Scale Y, sliderMax=5}
  )
  n34 = TextureCoordinate()
  n35 = AppendVector(
   A=ScalarParameter() {defaultValue=-1, group=Wispy Flame, parameterName=Distortion Speed X, sliderMax=5, sliderMin=-5}
   B=ScalarParameter() {group=Wispy Flame, parameterName=Distortion Speed Y, sliderMax=5, sliderMin=-5}
  )
  n36 = ScalarParameter() {defaultValue=1, group=Wispy Flame, parameterName=Distortion Scale X, sliderMax=5}
  n37 = ScalarParameter() {defaultValue=1, group=Wispy Flame, parameterName=Distortion Scale Y, sliderMax=5}
  n38 = TextureSampleParameter2D(UVs=Multiply(A=Panner(Coordinate=@n34, Speed=@n35) {speedX=1}, B=AppendVector(A=@n36, B=@n37))) {
   group=Wispy Flame
   parameterName=Distortion Texture
   samplerType=SAMPLERTYPE_LinearGrayscale
   texture=/Game/Effects/Fort_Effects/Emotes/UltraLight/Texture/T_FluidFlameTile_Greyscale
  }
  n39 = ScalarParameter() {group=Wispy Flame, parameterName=Distortion Strength, sliderMax=1}
  n40 = TextureSampleParameter2D(UVs=Add(A=Multiply(A=@n32, B=@n33), B=Multiply(A=@n38.R, B=@n39) {constB=0.1})) {
   group=Wispy Flame
   parameterName=Flame Texture
   samplerType=SAMPLERTYPE_Grayscale
   texture=/Game/Effects/Fort_Effects/Textures/Fire/T_WispyFireNoise_Soft_M
  }
  n41 = Multiply(
   A=Power(Base=@n19, Exp=ScalarParameter() {parameterName=End Fade}) {constExponent=0}
   B=LinearInterpolate(A=Multiply(A=@n28, B=@n40.R), B=@n27, Alpha=@n27)
  )
  n42 = Saturate(None=@n41)
  n43 = ScalarParameter() {defaultValue=1, group=Color, parameterName=Color Power, sliderMax=5}
  n44 = CurveAtlasRowParameter(CurveTime=SmoothStep(
    Min=ScalarParameter() {group=Color, parameterName=Color Smooth Step Min, sliderMax=1, sliderMin=-1}
    Max=ScalarParameter() {defaultValue=2, group=Color, parameterName=Color Smooth Step Max, sliderMax=5}
    Value=Power(Base=@n42, Exp=@n43) {constExponent=1}
   ) {constMax=2}) {
   atlas=&/CA_Emote_HornedJudgement
   curve=&/C_Emote_HornedJudgement
   parameterName=Curve_Color
  }
  MaterialFunctionCall(Emissive=MaterialFunctionCall(A=Multiply(A=@n13.RGB, B=@n14) {constB=2}, B=Multiply(A=@n13.RGB, B=@n15) {constB=1.5}, C=Multiply(A=@n16, B=@n17), Alpha=@n44) {materialFunction=&Lerp_3Color}.Result) {materialFunction=&MF_FX_EmissiveAdjustments}
  n45 = Multiply(A=ParticleColor().A, B=@n41)
  n46 = MaterialFunctionCall("FadeSensitivity (optional)"=ScalarParameter() {defaultValue=4, parameterName=ViewAngleFadeSensitivity, sliderMax=10}) {materialFunction=&FadeBasedOnViewAngle}
  n47 = DynamicParameter() {defaultValue=<rgba 0 0.75 0.8 1>, paramNames [UV Random Offset, Local Axis Fade Min, Side Fade, Swipe]}
  n48 = DynamicParameter() {parameterIndex=1, defaultValue=<rgba 0 0 1 1>, paramNames [Axis X, Axis Y, Axis Z, Param4]}
  n49 = MaterialFunctionCall(R=@n48.Axis X, G=@n48.Axis X, B=@n48.Axis X) {materialFunction=&AppendMany}
  n50 = CameraVectorWS()
  n51 = SmoothStep(Min=@n47.Local Axis Fade Min, Value=Abs(None=DotProduct(A=@n49.RGB, B=@n50)))
  n52 = Constant() {r=1}
  n53 = SmoothStep(
   Max=OneMinus(None=ScalarParameter() {defaultValue=0.9, parameterName=Edge Hardness, sliderMax=1})
   Value=Multiply(A=Multiply(A=@n45, B=@n46.Result), B=StaticSwitchParameter(True=@n51, False=@n52) {parameterName=Use local Axis Fade})
  ) {constMax=0.1}
  n54 = DepthFade(Opacity=@n53, FadeDistance=ScalarParameter() {defaultValue=35, parameterName=Depth Fade Distance, sliderMax=100}) {fadeDistanceDefault=25}
  ShadingPathSwitch(Default=@n54, Deferred=@n54, Forward=@n53, Mobile=@n53)
 ]
}
