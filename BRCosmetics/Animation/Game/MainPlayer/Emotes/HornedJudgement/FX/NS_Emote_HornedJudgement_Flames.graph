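# Niagara system for the Horned Judgement emote's blue flame effect, built from plasma mesh layers, splash meshes and ember sprites.
# Main, side and oar plasma emitters follow the system's skeletal mesh, with oar opacity also driven by a sampled bone's height.
# Plasma materials receive the offset between a rotated target and its smoothed position, giving motion lag.
# Delayed splashes and curl-noise embers add accents; mesh layers scale brightness by time of day.
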NiagaraSystem /BRCosmetics/Animation/Game/MainPlayer/Emotes/HornedJudgement/FX/NS_Emote_HornedJudgement_Flames
dependencies {
 usedDataInterfaces [
  &NiagaraDataInterfaceCurve
  &NiagaraDataInterfaceVectorCurve
  &NiagaraDataInterfaceArrayFloat2
  &NiagaraDataInterfaceColorCurve
  &NiagaraDataInterfaceSkeletalMesh
 ]
 usedDynamicInputs [&Add_Vector]
 usedModules [
  &self:SystemSpawnScript.SystemScriptSource.SystemScriptGraph.NiagaraNodeAssignment_0.SetVariables_#1
  &self:SystemSpawnScript.SystemScriptSource.SystemScriptGraph.NiagaraNodeAssignment_1.SetVariables_#2
  &self:SystemSpawnScript.SystemScriptSource.SystemScriptGraph.NiagaraNodeAssignment_2.SetVariables_#3
  &SystemState
 ]
}
emitters {
 Embers {
  emitterUpdateScript [
   &EmitterState(
    Inactive Response=::NewEnumerator1
    Life Cycle Mode=::NewEnumerator1
    Loop Behavior=::NewEnumerator1
    Loop Delay=$System.Delay
    Loop Duration Mode=::NewEnumerator1
    UseLoopDelay=true
   )
   &SpawnBurst_Instantaneous(
    Spawn Count=&RandomRangeInt(Minimum=50, Maximum=70)
    Use Spawn Probability=false
    UseLoopCountLimit=false
   )
  ]
  particleSpawnScript [
   /SetVariables(Particles.PerParticleRandom:NiagaraFloat=&RandomRangeFloat(Minimum=0, Maximum=1)) {isSetParameters=true}
   &FN_SetSpawnGroup(Number of Spawn Groups=2)
   &InitializeParticle(
    Color=<rgba 0 1.1 3 1>
    Lifetime=&Lerp_Float(A=0.5, B=1, Alpha=$Particles.PerParticleRandom)
    Sprite Size=&Multiply_Vector2D(
     A=&SelectVector2DFromArray(
      Direct Array Index=&SpawnGroup()
      Vector 2D Selection Array=DataInterface() {propertyValues {FloatData [<1 1>, <1 3>]}}
     )
     B=&MakeVector2D(
      X=&FloatFromCurve(
       FloatCurve=DataInterface() {propertyValues {Curve {keys [<at 0 1>,<at 0.8 2 RCIM_Constant>,<at 0.95 3 RCIM_Constant>]}}}
       CurveIndex=$Particles.PerParticleRandom
      )
      Y=&FloatFromCurve(
       FloatCurve=DataInterface() {propertyValues {Curve {keys [<at 0 1>,<at 0.8 2 RCIM_Constant>,<at 0.95 3>]}}}
       CurveIndex=$Particles.PerParticleRandom
      )
     )
    )
    Sprite Size Mode=::NewEnumerator4
    UsePositionOffset=false
    Write Material Random=true
   )
   &ShapeLocation(
    Box Size=<20 10 75>
    MaskSpawnGroup=false
    Offset=<0 0 10>
    Offset Mode=::NewEnumerator0
    Override Seed=false
    Rotation Axis=<0 1 0>
    Rotation Mode=::NewEnumerator1
    Shape Primitive=::NewEnumerator2
    Surface Only=false
    UseOwnerScale=true
   ) {module=ShapeLocation001}
   &AddVelocity(
    Module.Use Falloff Distance=false
    Origin Offset=<0 0 -5>
    Override Seed=false
    Velocity Mode=::NewEnumerator1
    Velocity Speed=&RandomRangeFloat(Minimum=100, Maximum=500)
   ) {module=AddVelocity001}
  ]
  particleUpdateScript [
   &ParticleState()
   &Drag(
    Drag=&FloatFromCurve(
     FloatCurve=DataInterface() {propertyValues {Curve {keys [{interpMode=RCIM_Cubic, tangentMode=RCTM_User, time=0, value=0.4}, {interpMode=RCIM_Cubic, tangentMode=RCTM_User, time=1, value=1}]}}}
     Scale Curve=&RandomRangeFloat(Minimum=8, Maximum=10)
    )
    Use Linear Drag=true
    Use Rotational Drag=false
   )
   &CurlNoiseForce(Noise Strength=500, Pan Noise=true, Pan Noise Field=<0 1 1>) {module=CurlNoiseForce001}
   &SolveForcesAndVelocity(Clamp Velocity=false, Limit Acceleration=false)
   &ScaleColor(
    Scale Alpha=&FloatFromCurve(FloatCurve=DataInterface() {propertyValues {Curve {keys [{interpMode=RCIM_Cubic, tangentMode=RCTM_User, time=0, value=1}, {interpMode=RCIM_Cubic, tangentMode=RCTM_User, time=1, value=0}]}}})
    Scale RGB=&MakeVectorFromLinearColorRGB(Color=&ColorFromCurve(ColorCurve=DataInterface() {propertyValues {
        AlphaCurve {keys [<at 0 1 RCIM_Cubic>]}
        BlueCurve {keys [<at 0.0039 5>,<at 1 0.576 RCIM_Cubic>]}
        GreenCurve {keys [<at 0.0039 5>,<at 1 0.576 RCIM_Cubic>]}
        RedCurve {keys [<at 0.0039 5>,<at 1 0.576 RCIM_Cubic>]}
       }}))
    ScaleA=true
    ScaleRGB=true
   )
   &ScaleSpriteSize(Uniform Curve Sprite Scale=DataInterface() {propertyValues {Curve {keys [
        {tangentMode=RCTM_User, time=0, value=1}
        {arriveTangent=-0.0405, interpMode=RCIM_Cubic, leaveTangent=-0.0405, tangentMode=RCTM_User, time=0.7009, value=1}
        {arriveTangent=-8.0089, interpMode=RCIM_Cubic, leaveTangent=-8.0089, tangentMode=RCTM_User, time=1, value=0}
       ]}}})
   &ScaleSpriteSizeBySpeed(
    Max Scale Factor=<1 2>
    Sample Scale Factor By Curve=true
    Scale Factor Curve=DataInterface() {propertyValues {Curve {keys [<at 0 0>, {arriveTangent=1.4989, interpMode=RCIM_Cubic, tangentMode=RCTM_Break, time=1, value=1}]}}}
    Velocity Threshold=250
   )
  ]
  rendererClasses [&NiagaraSpriteRendererProperties]
  renderers [{rendererClass=&NiagaraSpriteRendererProperties}]
 }
 Water_Plasma {
  emitterUpdateScript [
   &EmitterState(Inactive Response=::NewEnumerator1, Life Cycle Mode=::NewEnumerator1, Loop Duration=1, UseLoopDelay=false)
   &SpawnBurst_Instantaneous(Use Spawn Probability=false, UseLoopCountLimit=false)
   &SpawnRate(SpawnRate=2, Use Spawn Probability=false)
  ]
  particleSpawnScript [
   &InitializeParticle(
    Color=<rgba 0 0.4869 1 1>
    Lifetime=&RandomRangeFloat(Minimum=1.5, Maximum=2)
    Mesh Scale=<1.2 1.7 0.5>
    Mesh Scale Mode=::NewEnumerator4
    UsePositionOffset=true
    Write Material Random=true
   )
   &FN_SetSpawnGroup(Number of Spawn Groups=2)
   &DynamicMaterialParameters(
    Index 0 Param 1=&RandomRangeFloat(Minimum=0, Maximum=1)
    Index 0 Param 3=1
    Module.Index0Param0Write=true
    Module.Index0Param1Write=true
    Module.Index0Param2Write=true
    Module.Index0Param3Write=false
   )
  ]
  particleUpdateScript [
   &ParticleState()
   &SkeletalMeshLocation(
    Bone Sampling Mode=::NewEnumerator4
    MaskSpawnGroup=false
    Sampled Position Offset=&MakeVector(
     X=0
     Y=&RandomRangeFloat(Minimum=-50, Maximum=-60)
     Z=&RandomRangeFloat(Minimum=0, Maximum=2)
    )
    Skeletal Mesh=$System.Skeletal Mesh
   )
   &ScaleColor(
    Scale Alpha=&FloatFromCurve(FloatCurve=DataInterface() {propertyValues {Curve {keys [
         {interpMode=RCIM_Cubic, tangentMode=RCTM_User, time=0, value=0}
         {interpMode=RCIM_Cubic, tangentMode=RCTM_User, time=0.5, value=1}
         {tangentMode=RCTM_User, time=1, value=0}
        ]}}})
    Scale RGB=&VectorFromFloat(Value=&FNTimeOfDayScalar(Night Scale=0.5, Morning Scale=1, Evening Scale=1))
    ScaleA=true
    ScaleRGB=true
   )
   &InterpolateOverTime(
    Rate Of Change=8
    Target Vector Value:Vector3f=&AddVectorToPosition(
     Position=$Particles.Position
     Vector=&RotateVectorByQuaternion(Vector To Rotate=<0 -100 0>, Quaternion=$Particles.SkeletalMeshLocation.SampledRotation)
    )
    Variable Type=::NewEnumerator2
   )
   &DynamicMaterialParameters(
    Index 2 Param 1=&MakeFloatFromVector(
     Vector=&Subtract_Position(
      A=&AddVectorToPosition(
       Position=$Particles.Position
       Vector=&RotateVectorByQuaternion(Vector To Rotate=<0 -100 0>, Quaternion=$Particles.SkeletalMeshLocation.SampledRotation)
      )
      B=Linked() {linkedVariable {name=Particles.InterpolateOverTime.InterpolateOverTime_V3_.Moving Average, type=Vector3f}}
     )
     Channel=::NewEnumerator0
    )
    Index 2 Param 2=&MakeFloatFromVector(
     Vector=&Subtract_Position(
      A=&AddVectorToPosition(
       Position=$Particles.Position
       Vector=&RotateVectorByQuaternion(Vector To Rotate=<0 -100 0>, Quaternion=$Particles.SkeletalMeshLocation.SampledRotation)
      )
      B=Linked() {linkedVariable {name=Particles.InterpolateOverTime.InterpolateOverTime_V3_.Moving Average, type=Vector3f}}
     )
     Channel=::NewEnumerator1
    )
    Index 2 Param 3=&MakeFloatFromVector(
     Vector=&Subtract_Position(
      A=&AddVectorToPosition(
       Position=$Particles.Position
       Vector=&RotateVectorByQuaternion(Vector To Rotate=<0 -100 0>, Quaternion=$Particles.SkeletalMeshLocation.SampledRotation)
      )
      B=Linked() {linkedVariable {name=Particles.InterpolateOverTime.InterpolateOverTime_V3_.Moving Average, type=Vector3f}}
     )
     Channel=::NewEnumerator2
    )
    Module.Param0WriteEnabled2=true
    Module.Param1WriteEnabled2=true
    Module.Param2WriteEnabled2=true
    Module.Param3WriteEnabled2=true
    Write Parameter Index 0=false
    Write Parameter Index 2=true
   ) {module=DynamicMaterialParameters001}
   &ScaleMeshSize(Scale Factor=&VectorFromCurve(
     VectorCurve=DataInterface() {propertyValues {
       XCurve {keys [{arriveTangent=1.898, interpMode=RCIM_Cubic, leaveTangent=1.898, tangentMode=RCTM_User, time=0, value=1}]}
       YCurve {keys [{arriveTangent=0.5312, interpMode=RCIM_Cubic, leaveTangent=0.5312, tangentMode=RCTM_User, time=0, value=1}, <at 1 1.2>]}
       ZCurve {keys [{arriveTangent=9.0725, interpMode=RCIM_Cubic, leaveTangent=9.0725, tangentMode=RCTM_User, time=0, value=1}]}
      }}
     Scale Curve=<1 1 1.5>
    )) {module=ScaleMeshSize001}
  ]
  rendererClasses [&NiagaraMeshRendererProperties]
  renderers [{rendererClass=&NiagaraMeshRendererProperties, rendererIndex=0}]
 }
 Water_Plasma_Oars {
  emitterUpdateScript [
   &EmitterState(
    Inactive Response=::NewEnumerator1
    Life Cycle Mode=::NewEnumerator1
    Loop Behavior=::NewEnumerator1
    Loop Duration Mode=::NewEnumerator1
    UseLoopDelay=true
   )
   &SpawnBurst_Instantaneous(Use Spawn Probability=false, UseLoopCountLimit=false)
   &SpawnRate(SpawnRate=2, Use Spawn Probability=false)
  ]
  particleSpawnScript [
   &InitializeParticle(
    Color=<rgba 0 0.4869 1 1>
    Lifetime=&RandomRangeFloat(Minimum=1.5, Maximum=2)
    Mesh Scale=<0.6 0.7 0.5>
    Mesh Scale Mode=::NewEnumerator4
    UsePositionOffset=true
    Write Material Random=true
   )
   &FN_SetSpawnGroup(Number of Spawn Groups=2)
   &DynamicMaterialParameters(
    Index 0 Param 1=&RandomRangeFloat(Minimum=0, Maximum=1)
    Index 0 Param 3=1
    Module.Index0Param0Write=true
    Module.Index0Param1Write=true
    Module.Index0Param2Write=true
    Module.Index0Param3Write=false
   )
   &CameraOffset(Camera Offset Amount=3)
  ]
  particleUpdateScript [
   &ParticleState()
   &SkeletalMeshLocation(
    Bone Sampling Mode=::NewEnumerator4
    MaskSpawnGroup=false
    Position Sampling=::NewEnumerator2
    Sampled Position Offset=&MakeVector(
     X=&MaskFloatBySpawnGroup(Masked Float=-50, Passthrough Float=50)
     Y=-30
     Z=&RandomRangeFloat(Minimum=5, Maximum=6)
    )
    Skeletal Mesh=$System.Skeletal Mesh
   )
   &SkeletalMeshLocation(
    Bone / Socket Index=&Add_Integer(A=1, B=&SpawnGroup())
    Bone Sampling Mode=::NewEnumerator4
    MaskSpawnGroup=false
    Orientation Sampling=::NewEnumerator1
    Sampled Position Offset=&MaskVectorBySpawnGroup(Masked Vector=<23 0 0>, Passthrough Vector=<-23 0 0>)
    Skeletal Mesh=$System.Skeletal Mesh
   ) {module=SkeletalMeshLocation002}
   &SkeletalMeshLocation(
    Bone / Socket Index=1
    Bone Sampling Mode=::NewEnumerator4
    MaskSpawnGroup=false
    Orientation Sampling=::NewEnumerator1
    Position Sampling=::NewEnumerator2
    Sampled Position Coordinate Space=::Local
    Sampled Position Offset=<20 0 0>
    Skeletal Mesh=$System.Skeletal Mesh
   ) {module=SkeletalMeshLocation001}
   &ScaleColor(
    Scale Alpha=&FloatFromCurve(
     FloatCurve=DataInterface() {propertyValues {Curve {keys [
         {interpMode=RCIM_Cubic, tangentMode=RCTM_User, time=0, value=0}
         {interpMode=RCIM_Cubic, tangentMode=RCTM_User, time=0.5, value=1}
         {tangentMode=RCTM_User, time=1, value=0}
        ]}}}
     Scale Curve=&RemapRange(
      Input Value To Remap=&MakeFloatFromVector(Vector=$Particles.SkeletalMeshLocation001.SampledPosition, Channel=::NewEnumerator2)
      Input Min=0
      Input Max=13
      Output Min=1
      Output Max=0
     )
    )
    Scale RGB=&VectorFromFloat(Value=&FNTimeOfDayScalar(Night Scale=0.5, Morning Scale=1, Evening Scale=1))
    ScaleA=true
    ScaleRGB=true
   )
   &InterpolateOverTime(
    Rate Of Change=10
    Target Vector Value:Vector3f=&AddVectorToPosition(
     Position=$Particles.Position
     Vector=&RotateVectorByQuaternion(Vector To Rotate=<0 -100 0>, Quaternion=$Particles.SkeletalMeshLocation.SampledRotation)
    )
    Variable Type=::NewEnumerator2
   )
   &DynamicMaterialParameters(
    Index 2 Param 1=&MakeFloatFromVector(
     Vector=&Subtract_Position(
      A=&AddVectorToPosition(
       Position=$Particles.Position
       Vector=&RotateVectorByQuaternion(Vector To Rotate=<0 -100 0>, Quaternion=$Particles.SkeletalMeshLocation.SampledRotation)
      )
      B=Linked() {linkedVariable {name=Particles.InterpolateOverTime.InterpolateOverTime_V3_.Moving Average, type=Vector3f}}
     )
     Channel=::NewEnumerator0
    )
    Index 2 Param 2=&MakeFloatFromVector(
     Vector=&Subtract_Position(
      A=&AddVectorToPosition(
       Position=$Particles.Position
       Vector=&RotateVectorByQuaternion(Vector To Rotate=<0 -100 0>, Quaternion=$Particles.SkeletalMeshLocation.SampledRotation)
      )
      B=Linked() {linkedVariable {name=Particles.InterpolateOverTime.InterpolateOverTime_V3_.Moving Average, type=Vector3f}}
     )
     Channel=::NewEnumerator1
    )
    Index 2 Param 3=&MakeFloatFromVector(
     Vector=&Subtract_Position(
      A=&AddVectorToPosition(
       Position=$Particles.Position
       Vector=&RotateVectorByQuaternion(Vector To Rotate=<0 -100 0>, Quaternion=$Particles.SkeletalMeshLocation.SampledRotation)
      )
      B=Linked() {linkedVariable {name=Particles.InterpolateOverTime.InterpolateOverTime_V3_.Moving Average, type=Vector3f}}
     )
     Channel=::NewEnumerator2
    )
    Module.Param0WriteEnabled2=true
    Module.Param1WriteEnabled2=true
    Module.Param2WriteEnabled2=true
    Module.Param3WriteEnabled2=true
    Write Parameter Index 0=false
    Write Parameter Index 2=true
   ) {module=DynamicMaterialParameters001}
   &ScaleMeshSize(Scale Factor=&VectorFromCurve(
     VectorCurve=DataInterface() {propertyValues {
       XCurve {keys [{arriveTangent=1.898, interpMode=RCIM_Cubic, leaveTangent=1.898, tangentMode=RCTM_User, time=0, value=1}]}
       YCurve {keys [{arriveTangent=0.5312, interpMode=RCIM_Cubic, leaveTangent=0.5312, tangentMode=RCTM_User, time=0, value=1}, <at 1 1.2>]}
       ZCurve {keys [{arriveTangent=9.0725, interpMode=RCIM_Cubic, leaveTangent=9.0725, tangentMode=RCTM_User, time=0, value=1}]}
      }}
     Scale Curve=<1 1 1.5>
    )) {module=ScaleMeshSize001}
  ]
  rendererClasses [&NiagaraMeshRendererProperties]
  renderers [{rendererClass=&NiagaraMeshRendererProperties, rendererIndex=0}]
 }
 Water_Plasma_Side {
  emitterUpdateScript [
   &EmitterState(Inactive Response=::NewEnumerator1, Life Cycle Mode=::NewEnumerator1, Loop Duration=1, UseLoopDelay=false)
   &SpawnBurst_Instantaneous(Use Spawn Probability=false, UseLoopCountLimit=false)
   &SpawnRate(SpawnRate=2, Use Spawn Probability=false)
  ]
  particleSpawnScript [
   &InitializeParticle(
    Color=<rgba 0 0.4869 1 1>
    Lifetime=&RandomRangeFloat(Minimum=1.5, Maximum=2)
    Mesh Scale=<1 1.4 1.2>
    Mesh Scale Mode=::NewEnumerator4
    UsePositionOffset=false
    Write Material Random=true
   )
   &FN_SetSpawnGroup(Number of Spawn Groups=2)
   &DynamicMaterialParameters(
    Index 0 Param 1=&RandomRangeFloat(Minimum=0, Maximum=1)
    Index 0 Param 3=0.25
    Module.Index0Param0Write=true
    Module.Index0Param1Write=true
    Module.Index0Param2Write=true
    Module.Index0Param3Write=false
   )
   &CameraOffset(Camera Offset Amount=5)
  ]
  particleUpdateScript [
   &ParticleState()
   &SkeletalMeshLocation(
    Bone Sampling Mode=::NewEnumerator4
    MaskSpawnGroup=false
    Sampled Position Offset=&MakeVector(X=0, Y=-55, Z=-5)
    Skeletal Mesh=$System.Skeletal Mesh
   )
   &SampleSkeletalMesh(
    AdditionalYawPitchRoll=false
    Bone Sampling Mode:ENiagaraBoneSamplingMode=::NewEnumerator4
    Sampled Orientation Coordinate Space:ENiagaraCoordinateSpace=::World
    Sampled Position Coordinate Space:ENiagaraCoordinateSpace=::World
    Skeletal Mesh:NiagaraDataInterfaceSkeletalMesh=$System.Skeletal Mesh
   )
   &ScaleColor(
    Scale Alpha=&FloatFromCurve(FloatCurve=DataInterface() {propertyValues {Curve {keys [
         {interpMode=RCIM_Cubic, tangentMode=RCTM_User, time=0, value=0}
         {interpMode=RCIM_Cubic, tangentMode=RCTM_User, time=0.5, value=1}
         {tangentMode=RCTM_User, time=1, value=0}
        ]}}})
    Scale RGB=&VectorFromFloat(Value=&FNTimeOfDayScalar(Night Scale=0.5, Morning Scale=1, Evening Scale=1))
    ScaleA=true
    ScaleRGB=true
   )
   &InterpolateOverTime(
    Rate Of Change=8
    Target Vector Value:Vector3f=&AddVectorToPosition(
     Position=$Particles.SampleSkeletalMesh.SampledPosition
     Vector=&RotateVectorByQuaternion(Vector To Rotate=<0 -100 0>, Quaternion=$Particles.SampleSkeletalMesh.SampledRotation)
    )
    Variable Type=::NewEnumerator2
   )
   &DynamicMaterialParameters(
    Index 2 Param 1=&MakeFloatFromVector(
     Vector=&Subtract_Position(
      A=&AddVectorToPosition(
       Position=$Particles.SampleSkeletalMesh.SampledPosition
       Vector=&RotateVectorByQuaternion(Vector To Rotate=<0 -100 0>, Quaternion=$Particles.SampleSkeletalMesh.SampledRotation)
      )
      B=Linked() {linkedVariable {name=Particles.InterpolateOverTime.InterpolateOverTime_V3_.Moving Average, type=Vector3f}}
     )
     Channel=::NewEnumerator0
    )
    Index 2 Param 2=&MakeFloatFromVector(
     Vector=&Subtract_Position(
      A=&AddVectorToPosition(
       Position=$Particles.SampleSkeletalMesh.SampledPosition
       Vector=&RotateVectorByQuaternion(Vector To Rotate=<0 -100 0>, Quaternion=$Particles.SampleSkeletalMesh.SampledRotation)
      )
      B=Linked() {linkedVariable {name=Particles.InterpolateOverTime.InterpolateOverTime_V3_.Moving Average, type=Vector3f}}
     )
     Channel=::NewEnumerator1
    )
    Index 2 Param 3=&MakeFloatFromVector(
     Vector=&Subtract_Position(
      A=&AddVectorToPosition(
       Position=$Particles.SampleSkeletalMesh.SampledPosition
       Vector=&RotateVectorByQuaternion(Vector To Rotate=<0 -100 0>, Quaternion=$Particles.SampleSkeletalMesh.SampledRotation)
      )
      B=Linked() {linkedVariable {name=Particles.InterpolateOverTime.InterpolateOverTime_V3_.Moving Average, type=Vector3f}}
     )
     Channel=::NewEnumerator2
    )
    Module.Param0WriteEnabled2=true
    Module.Param1WriteEnabled2=true
    Module.Param2WriteEnabled2=true
    Module.Param3WriteEnabled2=true
    Write Parameter Index 0=false
    Write Parameter Index 2=true
   ) {module=DynamicMaterialParameters001}
   &ScaleMeshSize(Scale Factor=&VectorFromCurve(VectorCurve=DataInterface() {propertyValues {
       XCurve {keys [{arriveTangent=1.898, interpMode=RCIM_Cubic, leaveTangent=1.898, tangentMode=RCTM_User, time=0, value=1}]}
       YCurve {keys [{arriveTangent=0.5312, interpMode=RCIM_Cubic, leaveTangent=0.5312, tangentMode=RCTM_User, time=0, value=1}, <at 1 1.2>]}
       ZCurve {keys [{arriveTangent=9.0725, interpMode=RCIM_Cubic, leaveTangent=9.0725, tangentMode=RCTM_User, time=0, value=1}]}
      }})) {module=ScaleMeshSize001}
  ]
  rendererClasses [&NiagaraMeshRendererProperties]
  renderers [{rendererClass=&NiagaraMeshRendererProperties, rendererIndex=0}]
 }
 Water_Plasma_Splash001 {
  emitterUpdateScript [
   &EmitterState(Life Cycle Mode=::NewEnumerator1, Loop Behavior=::NewEnumerator1, Loop Delay=$System.Delay, Loop Duration=0.25, UseLoopDelay=true)
   &SpawnRate(
    SpawnRate=&FloatFromCurve(
     FloatCurve=DataInterface() {propertyValues {Curve {keys [{interpMode=RCIM_Cubic, tangentMode=RCTM_User, time=0, value=1}, {interpMode=RCIM_Cubic, tangentMode=RCTM_User, time=1, value=0}]}}}
     CurveIndex=$Emitter.NormalizedLoopAge
     Scale Curve=60
    )
    Use Spawn Probability=false
   )
  ]
  particleSpawnScript [
   &FN_SetSpawnGroup(Number of Spawn Groups=2)
   &InitializeParticle(
    Color=<rgba 0 0.55 1.5 0.8>
    Lifetime=&RandomRangeFloat(Minimum=0.4, Maximum=0.8)
    Mesh Scale Max=<0.5 0.8 0.4>
    Mesh Scale Min=<0.4 0.5 0.4>
    Mesh Scale Mode=::NewEnumerator2
    Position Offset=&MakeVector(
     X=&MaskFloatBySpawnGroup(Masked Float=19, Passthrough Float=-19)
     Y=&RandomRangeFloat(Minimum=-15, Maximum=15)
     Z=0
    )
    UsePositionOffset=true
    Write Material Random=true
   )
   &DynamicMaterialParameters(
    Index 0 Param 1=&RandomRangeFloat(Minimum=0, Maximum=1)
    Module.Index0Param0Write=true
    Module.Index0Param1Write=true
    Module.Index0Param2Write=true
    Module.Index0Param3Write=true
   )
   &AddVelocity(
    Module.Use Falloff Distance=false
    Origin Offset=<0 0 -2>
    Override Seed=false
    Velocity Mode=::NewEnumerator1
    Velocity Speed=&RandomRangeFloat(Minimum=10, Maximum=20)
   ) {enabled=false, module=AddVelocity001}
   &AddVelocity(
    Velocity=&MaskVectorBySpawnGroup(Masked Vector=<1 0 0>, Passthrough Vector=<-1 0 0>)
    Velocity Speed Scale=&RandomRangeFloat(Minimum=20, Maximum=30)
   ) {module=AddVelocity002}
  ]
  particleUpdateScript [
   &ParticleState()
   &GravityForce(Gravity=<0 0 -10>)
   &Drag(Drag=2, Use Linear Drag=true, Use Rotational Drag=true)
   &ScaleColor(
    Scale Alpha=&FloatFromCurve(
     FloatCurve=DataInterface() {propertyValues {Curve {keys [
         {arriveTangent=9.8898, interpMode=RCIM_Cubic, leaveTangent=9.8898, tangentMode=RCTM_User, time=0, value=0}
         {interpMode=RCIM_Cubic, tangentMode=RCTM_User, time=0.3, value=1}
         {arriveTangent=-3.6098, tangentMode=RCTM_User, time=1, value=0}
        ]}}}
     Scale Curve=2
    )
    Scale RGB=&VectorFromFloat(Value=&FloatFromCurve(
      FloatCurve=DataInterface() {propertyValues {Curve {keys [
          {arriveTangent=9.0725, interpMode=RCIM_Cubic, leaveTangent=9.0725, tangentMode=RCTM_User, time=0, value=0.5}
          {arriveTangent=-0.7694, interpMode=RCIM_Cubic, leaveTangent=-0.7694, tangentMode=RCTM_User, time=0.1408, value=0.9847}
          {arriveTangent=-0.0405, interpMode=RCIM_Cubic, leaveTangent=-0.0405, tangentMode=RCTM_User, time=1, value=0.5}
         ]}}}
      Scale Curve=&FNTimeOfDayScalar(Night Scale=0.5, Morning Scale=1, Evening Scale=1)
     ))
    ScaleA=true
    ScaleRGB=true
   )
   &InterpolateOverTime(
    Rate Of Change=10
    Target Vector Value:Vector3f=&AddVectorToPosition(
     Position=$Particles.Position
     Vector=&RotateVectorByQuaternion(Vector To Rotate=<0 -200 0>, Quaternion=$Particles.SkeletalMeshLocation.SampledRotation)
    )
    Variable Type=::NewEnumerator2
   )
   &DynamicMaterialParameters(
    Index 0 Param 3=&FloatFromCurve(
     FloatCurve=DataInterface() {propertyValues {Curve {keys [{interpMode=RCIM_Cubic, tangentMode=RCTM_User, time=0, value=0.5}, {interpMode=RCIM_Cubic, tangentMode=RCTM_User, time=1, value=1}]}}}
     Scale Curve=&RandomRangeFloat(Minimum=1, Maximum=2)
    )
    Index 0 Param 4=&FloatFromCurve(
     FloatCurve=DataInterface() {propertyValues {Curve {keys [
         {arriveTangent=-1.9474, interpMode=RCIM_Cubic, leaveTangent=-1.9474, tangentMode=RCTM_User, time=0, value=1}
         {interpMode=RCIM_Cubic, tangentMode=RCTM_User, time=1, value=0}
        ]}}}
     Scale Curve=0.8
    )
    Index 2 Param 1=&MakeFloatFromVector(
     Vector=&Subtract_Position(
      A=&AddVectorToPosition(
       Position=$Particles.Position
       Vector=&RotateVectorByQuaternion(Vector To Rotate=<0 -200 0>, Quaternion=$Particles.SkeletalMeshLocation.SampledRotation)
      )
      B=Linked() {linkedVariable {name=Particles.InterpolateOverTime.InterpolateOverTime_V3_.Moving Average, type=Vector3f}}
     )
     Channel=::NewEnumerator0
    )
    Index 2 Param 2=&MakeFloatFromVector(
     Vector=&Subtract_Position(
      A=&AddVectorToPosition(
       Position=$Particles.Position
       Vector=&RotateVectorByQuaternion(Vector To Rotate=<0 -200 0>, Quaternion=$Particles.SkeletalMeshLocation.SampledRotation)
      )
      B=Linked() {linkedVariable {name=Particles.InterpolateOverTime.InterpolateOverTime_V3_.Moving Average, type=Vector3f}}
     )
     Channel=::NewEnumerator1
    )
    Index 2 Param 3=&MakeFloatFromVector(
     Vector=&Subtract_Position(
      A=&AddVectorToPosition(
       Position=$Particles.Position
       Vector=&RotateVectorByQuaternion(Vector To Rotate=<0 -200 0>, Quaternion=$Particles.SkeletalMeshLocation.SampledRotation)
      )
      B=Linked() {linkedVariable {name=Particles.InterpolateOverTime.InterpolateOverTime_V3_.Moving Average, type=Vector3f}}
     )
     Channel=::NewEnumerator2
    )
    Module.Param0WriteEnabled=false
    Module.Param0WriteEnabled2=true
    Module.Param1WriteEnabled=false
    Module.Param1WriteEnabled2=true
    Module.Param2WriteEnabled=true
    Module.Param2WriteEnabled2=true
    Module.Param3WriteEnabled=true
    Module.Param3WriteEnabled2=true
    Write Parameter Index 2=true
   ) {module=DynamicMaterialParameters001}
   &ScaleMeshSize(Scale Factor=&VectorFromCurve(VectorCurve=DataInterface() {propertyValues {
       XCurve {keys [
         {arriveTangent=4.3367, interpMode=RCIM_Cubic, leaveTangent=4.3367, tangentMode=RCTM_User, time=0, value=1}
         {arriveTangent=0.25, interpMode=RCIM_Cubic, leaveTangent=0.25, time=0.4, value=1.65}
         <at 1 1.25>
        ]}
       YCurve {keys [
         {arriveTangent=1.6312, interpMode=RCIM_Cubic, leaveTangent=1.6312, tangentMode=RCTM_User, time=0, value=1}
         <at 0.3 1.2 RCIM_Cubic>
         <at 1 1>
        ]}
       ZCurve {keys [
         {arriveTangent=1.2374, interpMode=RCIM_Cubic, leaveTangent=1.2374, tangentMode=RCTM_User, time=0, value=0.9}
         <at 0.3 1.3 RCIM_Cubic>
         <at 1 0.9 RCIM_Cubic>
        ]}
      }}))
   &SolveForcesAndVelocity(Clamp Velocity=false, Limit Acceleration=false)
  ]
  rendererClasses [&NiagaraMeshRendererProperties]
  renderers [{rendererClass=&NiagaraMeshRendererProperties, rendererIndex=0}]
 }
}
summary {
 systemName=NS_Emote_HornedJudgement_Flames
 emitters [
  {emitterName=Water_Plasma, rendererClasses [&NiagaraMeshRendererProperties]}
  {emitterName=Water_Plasma_Side, rendererClasses [&NiagaraMeshRendererProperties]}
  {emitterName=Water_Plasma_Splash001, rendererClasses [&NiagaraMeshRendererProperties]}
  {emitterName=Embers, rendererClasses [&NiagaraSpriteRendererProperties]}
  {emitterName=Water_Plasma_Oars, rendererClasses [&NiagaraMeshRendererProperties]}
 ]
 userVariables [
  User.AdvancedEmissiveColor {type=LinearColor, description="", defaultValue {struct=LinearColor, value=<rgba 0.0194 0.526 0.4838 1>}}
  User.bIsFrontEnd {type=NiagaraBool}
  User.bIsFrontEndPreview {type=NiagaraBool}
  User.NormalizedNotifyProgress {type=NiagaraFloat, defaultValue=0, description=""}
  User.RibbonIDOffset {type=NiagaraInt32}
 ]
}
